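; Gameplay ability that forces a player through the smash interaction on a locking display case in Heist Bases.
; On activation it applies an effect preventing interaction cancellation, disables input and plays the display case montage.
; The montage trigger calls FinishProcess on the display case, found as original instigator of the granting effect's context.
; Cleanup removes the effect and restores input, repeated in OnEndAbility as a failsafe.
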(event Ability|EventFailedToActivatePassiveAbility
  (params (ActorInfo :type Gameplay%20Ability%20Actor%20Info%20Structure)))

(event Ability|EventActivateAbility
  (bind (_returnvalue_owneractor _returnvalue_avataractor _returnvalue_playercontroller _returnvalue_abilitysystemcomponent) (Utilities|Struct|BreakGameplayAbilityActorInfo (Ability|GetActorInfo)))
  (bind _asfort_player_pawn (Utilities|Casting|CastToFortPlayerPawn _returnvalue_avataractor))
  (bind _output_get (Variables|Default|SetPlayerPawn _asfort_player_pawn))
  (bind _returnvalue (GameplayEffects|ApplyGameplayEffectToSelf _returnvalue_abilitysystemcomponent "/HeistBasesGameplay/GameplayAbility/GE_Prevent_Interact_Cancellation.GE_Prevent_Interact_Cancellation_C" 0.0 (Class|AthenaFunctionLibrary|MakeSimpleFortEffectContext (Ability|EffectContext|GetOriginalInstigatorActor (Ability|GetGrantedbyEffectContext)))))
  (Variables|Default|SetActionBlockingGE _returnvalue)
  (Input|DisableInput _output_get _returnvalue_playercontroller) ; Disallow input during montage
  (Ability|Tasks|PlayMontageInfoAndWaitForTrigger "None" (Variables|Default|GetDisplayCaseMontageInfo)
    (:Triggered
      ;[0 0.1 1] Montage trigger which tells interacting player to complete the smash
      (bind _aslocking_display_case (Utilities|Casting|CastToLockingDisplayCase (Ability|EffectContext|GetOriginalInstigatorActor (Ability|GetGrantedbyEffectContext))))
      (Class|LockingDisplayCase|FinishProcess _aslocking_display_case (Variables|Default|GetPlayerPawn)))
    (:Cancelled
      _)
    (:Completed
      _))
  ;[1 0.2 0] Make sure we return the player to normal state no matter how this GA ends. (GE has a duration for extra failsafe)
  (Ability|RemoveGameplayEffectFromOwnerWithHandle (Variables|Default|GetActionBlockingGE))
  (Input|EnableInput (Variables|Default|GetPlayerPawn) _returnvalue_playercontroller)
  (Ability|EndAbility))

(event Ability|EventOnEndAbility
  (params (bWasCancelled :type Boolean))
  (bind (_returnvalue_owneractor _returnvalue_avataractor _returnvalue_playercontroller) (Utilities|Struct|BreakGameplayAbilityActorInfo (Ability|GetActorInfo)))
  ;[1 0.2 0] Make sure we return the player to normal state no matter how this GA ends. (GE has a duration for extra failsafe)
  (Ability|RemoveGameplayEffectFromOwnerWithHandle (Variables|Default|GetActionBlockingGE))
  (Input|EnableInput (Variables|Default|GetPlayerPawn) _returnvalue_playercontroller)
  (Ability|EndAbility))
