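; Gameplay ability for interact-searching a display case in Heist bases.
; On activation it caches the avatar as the player pawn and calls StartInteractSearch.
; If the pawn has the override-montage tag, it plays the PlayerPawn_Athena's override search montage, or the display case montage when that is invalid, and calls EndAbility.
; Without the tag, or on a failed Athena cast, this graph plays no montage and never calls EndAbility.
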(event Ability|EventActivateAbility
  (bind (_returnvalue_owneractor _returnvalue_avataractor) (Utilities|Struct|BreakGameplayAbilityActorInfo (Ability|GetActorInfo)))
  (bind _asfort_player_pawn (Utilities|Casting|CastToFortPlayerPawn _returnvalue_avataractor))
  (Variables|Default|SetPlayerPawn _asfort_player_pawn)
  (CallFunction|StartInteractSearch)
  (if (GameplayTags|HasMatchingGameplayTag (Variables|Default|GetPlayerPawn) (Variables|Default|GetTag_UseOverrideSearchMontage))
    (bind _asplayer_pawn_athena (Utilities|Casting|CastToPlayerPawn_Athena (Variables|Default|GetPlayerPawn))
      (:then
        (Variables|Default|SetOverrideSearchMontageInfo (Variables|/Game/Athena/PlayerPawn_Athena.PlayerPawn_Athena_C|GetOverrideSearchMontageInfo _asplayer_pawn_athena))
        (Utilities|IsValid (Utilities|Struct|BreakFortGameplayAbilityMontageInfo (Variables|Default|GetOverrideSearchMontageInfo))
          (:Is%20Valid
            (Ability|Tasks|PlayMontageInfoAndWaitForTrigger "None" (Variables|Default|GetOverrideSearchMontageInfo)
              (:Cancelled
                _)))
          (:Is%20Not%20Valid
            (Ability|Tasks|PlayMontageInfoAndWaitForTrigger "None" (Variables|Default|GetDisplayCaseMontageInfo)
              (:Cancelled
                _))))
        (Ability|EndAbility))
      (:CastFailed
        _))
    (else
      _)))
